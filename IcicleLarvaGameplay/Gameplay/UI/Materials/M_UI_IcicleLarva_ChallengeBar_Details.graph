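# Translucent UI material that draws decorative angled detail shapes, apparently for the IcicleLarva challenge bar.
# Each half is a skewed box fill minus a narrow, slightly skewed box, with opposite skews mirroring left and right.
# A rounded U coordinate confines each shape to its own side before Max merges them.
# The merged mask is multiplied by a reflected gradient and halved, alongside a constant light-blue color.
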Material /IcicleLarvaGameplay/Gameplay/UI/Materials/M_UI_IcicleLarva_ChallengeBar_Details
graph {
 properties {bAutomaticallySetUsageInEditor=false, blendMode=BLEND_Translucent, materialDomain=MD_UI, usedShadingModels=Default Lit}
 nodes [
  Constant3Vector() {constant=<rgba 0.159 0.491 0.9823 1>}
  n1 = MaterialFunctionCall!(
   UV=Add(
    A=MaterialFunctionCall() {materialFunction=&GetUserInterfaceUV}.Normalized UV
    B=Constant2Vector() {g=0.14, r=-0.4}
   ) {constB=-0.2}
   Skew Angle=Constant() {r=50}
  ) {materialFunction=/FortUILibrary/MaterialFunctions/Basics/MF_UI_Skew}
  n2 = Constant2Vector() {g=0.36, r=3}
  n3 = MaterialFunctionCall(
   UV=Add(A=MaterialFunctionCall() {materialFunction=&GetUserInterfaceUV}.Normalized UV, B=Constant2Vector() {r=0.007})
   Skew Angle=Constant() {r=1}
  ) {materialFunction=&MF_UI_Skew}
  n4 = Constant2Vector() {g=1, r=0.08}
  n5 = Round(None=ComponentMask(None=TextureCoordinate()) {r=true})
  n6 = Subtract(
   A=Subtract(
    A=MaterialFunctionCall(UVs=@n1.Skewed UVs, Size=@n2) {materialFunction=&MF_UI_Box}.Fill
    B=MaterialFunctionCall(UVs=@n3.Skewed UVs, Size=@n4) {materialFunction=&MF_UI_Box}.Fill
   )
   B=OneMinus(None=@n5)
  )
  n7 = MaterialFunctionCall(
   UV=Add(
    A=MaterialFunctionCall() {materialFunction=&GetUserInterfaceUV}.Normalized UV
    B=Constant2Vector() {g=0.14, r=0.4}
   ) {constB=0}
   Skew Angle=Constant() {r=-50}
  ) {materialFunction=&MF_UI_Skew}
  n8 = Constant2Vector() {g=0.36, r=3}
  n9 = MaterialFunctionCall(
   UV=Add(A=MaterialFunctionCall() {materialFunction=&GetUserInterfaceUV}.Normalized UV, B=Constant2Vector() {r=-0.007})
   Skew Angle=Constant() {r=-1}
  ) {materialFunction=&MF_UI_Skew}
  n10 = Constant2Vector() {g=1, r=0.08}
  n11 = Subtract(
   A=Subtract(
    A=MaterialFunctionCall(UVs=@n7.Skewed UVs, Size=@n8) {materialFunction=&MF_UI_Box}.Fill
    B=MaterialFunctionCall(UVs=@n9.Skewed UVs, Size=@n10) {materialFunction=&MF_UI_Box}.Fill
   )
   B=@n5
  )
  Multiply(A=Multiply(
    A=Max(A=@n6, B=@n11)
    B=MaterialFunctionCall!(UVs, Offset, RotationCenter, RotationAngle) {materialFunction=/FortUILibrary/MaterialFunctions/Gradients/MF_UI_ReflectedGradient}.Result
   )) {constB=0.5}
 ]
}
