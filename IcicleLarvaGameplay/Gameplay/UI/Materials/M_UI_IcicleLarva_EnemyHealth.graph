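# Translucent UI material for the IcicleLarva enemy health bar, with a progress fill, edge marker, drop shadow and bracket coloring.
# A texture's R and G channels act as antialiased distance fields for shape and shadow, with a static Flip switch.
# The Progress parameter drives rotated-UV hard edges that form the fill and marker.
# Fill color is solid or a switchable reflected gradient, optionally overlaid with panning arrows whose opacity pulses over time.
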Material /IcicleLarvaGameplay/Gameplay/UI/Materials/M_UI_IcicleLarva_EnemyHealth
graph {
 properties {blendMode=BLEND_Translucent, materialDomain=MD_UI, pixelDepthOffsetMode=PDOM_Legacy, usedShadingModels=Default Lit}
 nodes [
  n1 = VectorParameter() {group=Color, parameterName=Backing_Color, defaultValue=<rgba 0 0 0 1>}
  n2 = NamedRerouteUsage()
  n3 = NamedRerouteUsage()
  n4 = LinearInterpolate(
   A=ScalarParameter() {defaultValue=0.05, group=Arrows, parameterName=Arrows_MaxOpacity}
   B=ScalarParameter() {defaultValue=0.01, group=Arrows, parameterName=Arrows_MinOpacity}
   Alpha=MaterialFunctionCall(Speed=Constant() {r=1}) {materialFunction=&DebugTimeSine}.Result
  ) {constA=0.05, constB=0.01}
  n5 = LinearInterpolate(
   A=@n2
   B=LinearInterpolate(A=@n2, B=VectorParameter() {group=Color, parameterName=ArrowsColor, defaultValue=<rgba 1 1 1 1>}.RGB, Alpha=Multiply(A=@n3, B=@n4))
   Alpha=ScalarParameter() {parameterName=AnimatedArrows}
  )
  n6 = NamedRerouteUsage()
  n7 = NamedRerouteUsage()
  LinearInterpolate(
   A=VectorParameter() {group=Color, parameterName=Brackets_Color, defaultValue=<rgba 0.0296 0.0409 0.0666 1>}.R
   B=LinearInterpolate(
    A=LinearInterpolate(A=@n1.RGB, B=@n5, Alpha=@n6)
    B=VectorParameter() {group=Color, parameterName=MarkerColor, defaultValue=<rgba 1 1 1 1>}.RGB
    Alpha=NamedRerouteUsage()
   )
   Alpha=@n7
  )
  n8 = Subtract(A=@n7, B=NamedRerouteUsage())
  n9 = ScalarParameter() {defaultValue=0.2, group=Bar, parameterName=Progress_Backing_Alpha}
  n10 = ScalarParameter() {defaultValue=1, group=Bar, parameterName=Progress_Fill_Alpha}
  n11 = NamedRerouteUsage()
  LinearInterpolate(
   A=Multiply(A=Saturate(None=@n8), B=LinearInterpolate(A=@n9, B=@n10, Alpha=@n11) {constAlpha=1})
   B=ScalarParameter() {defaultValue=0.4, group=Bar, parameterName=Shadow_Alpha}
   Alpha=NamedRerouteUsage()
  )
  n12 = ConstantBiasScale(None=TextureCoordinate()) {bias=-0.5, scale=2}
  n13 = Multiply(A=Subtract(A=Abs(None=@n12), B=Constant2Vector() {g=1, r=0.99}) {constB=0.5}) {constB=80}
  NamedRerouteDeclaration(None=Length(None=Max(A=@n13) {constB=0}) {is=class MaterialExpressionLength}) {name=SafeZone, nodeColor=<rgba 0 0 0 1>}
  n14 = Add(A=MaterialFunctionCall() {materialFunction=&GetUserInterfaceUV}.Normalized UV) {constB=0.008}
  n15 = Constant() {r=-0.0025}
  n16 = Subtract(A=OneMinus(None=MaterialFunctionCall(UVs=@n14, "Rotation Angle (0-1)"=@n15) {materialFunction=&CustomRotator}.Rotated Values))
  n17 = LinearInterpolate(
   A=ScalarParameter() {defaultValue=0.02, parameterName=Min}
   B=ScalarParameter() {defaultValue=0.97, parameterName=Max}
   Alpha=ScalarParameter() {defaultValue=0.7, group=Bar, parameterName=Progress}
  ) {constA=0.02, constB=0.97}
  n18 = Multiply(A=ComponentMask(None=Add(A=@n16, B=@n17)) {r=true}, B=ScalarParameter() {defaultValue=800, group=Bar, parameterName=Progress_Edge_Hardness})
  NamedRerouteDeclaration(None=Subtract(A=NamedRerouteUsage(), B=Saturate(None=@n18))) {name=BarMarker, nodeColor=<rgba 0 0 0 1>}
  n19 = VectorParameter() {group=Color, parameterName=GradientColor01, defaultValue=<rgba 0.6376 0.0232 0.0307 1>}
  n20 = VectorParameter() {group=Color, parameterName=GradientColor02, defaultValue=<rgba 1 1 1 1>}
  n21 = MaterialFunctionCall!(
   UVs=MaterialFunctionCall() {materialFunction=&GetUserInterfaceUV}.Normalized UV
   Offset
   RotationCenter
   RotationAngle
  ) {materialFunction=/FortUILibrary/MaterialFunctions/Gradients/MF_UI_ReflectedGradient}
  NamedRerouteDeclaration(None=StaticSwitchParameter(True=LinearInterpolate(A=@n19.RGB, B=@n20.RGB, Alpha=@n21.Result), False=@n19.RGB) {group=Color, parameterName=Gradient?, sortPriority=1}) {name=Color, nodeColor=<rgba 0 0 0 1>}
  n22 = TextureCoordinate()
  n23 = Multiply(A=ScalarParameter() {defaultValue=20, group=Arrows, parameterName=ArrowsDirection}, B=ComponentMask(None=@n22) {r=true})
  n24 = ComponentMask(None=@n22) {g=true}
  n25 = ScalarParameter() {defaultValue=-1.5, group=Arrows, parameterName=ArrowsSpeed}
  n26 = Constant()
  n27 = MaterialFunctionCall(
   UVs=Panner(Coordinate=AppendVector(A=@n23, B=@n24), Speed=AppendVector(A=@n25, B=@n26)) {speedX=0.1}
   "Rotation Angle (0-1)"=Constant() {r=0.25}
  ) {materialFunction=&CustomRotator}
  n28 = ComponentMask(None=Constant3Vector() {constant=<rgba 0.75 0.28 0 1>}) {g=true, r=true}
  n29 = Constant() {r=-0.06}
  n30 = ComponentMask(None=Constant3Vector() {constant=<rgba 6 1 0 1>}) {g=true, r=true}
  n31 = StaticBool() {value=true}
  NamedRerouteDeclaration(None=MaterialFunctionCall(
    UVs=MaterialFunctionCall!(UV=@n27.Rotated Values, Offset=@n28, Rotation=@n29, Scale=@n30, MirrorY=@n31) {materialFunction=/FortUILibrary/MaterialFunctions/Basics/MF_UI_Symmetry_2x1_DEPRECATED}.Result
    GlowMax=Constant() {r=0.01}
   ) {materialFunction=&MF_UI_Box}.Fill) {name=Arrows, nodeColor=<rgba 0 0 0 1>}
  n32 = OneMinus(None=MaterialFunctionCall(
    UVs=MaterialFunctionCall() {materialFunction=&GetUserInterfaceUV}.Normalized UV
    "Rotation Angle (0-1)"=Constant() {r=-0.004}
   ) {materialFunction=&CustomRotator}.Rotated Values)
  n33 = ScalarParameter() {defaultValue=0.02, parameterName=Min}
  n34 = ScalarParameter() {defaultValue=0.97, parameterName=Max}
  n35 = ScalarParameter() {defaultValue=0.7, group=Bar, parameterName=Progress}
  n36 = ComponentMask(None=Add(A=Subtract(A=@n32), B=LinearInterpolate(A=@n33, B=@n34, Alpha=@n35) {constA=0.02, constB=0.97})) {r=true}
  n37 = ScalarParameter() {defaultValue=800, group=Bar, parameterName=Progress_Edge_Hardness}
  NamedRerouteDeclaration(None=Saturate(None=Multiply(A=@n36, B=@n37))) {name=Progress Fill, nodeColor=<rgba 0 0 0 1>}
  n38 = ScalarParameter() {defaultValue=0.1, parameterName=XOffset}
  n39 = Constant()
  n40 = ScalarParameter() {defaultValue=0.8, parameterName=Scale}
  n41 = AppendVector(A=@n40, B=Constant() {r=-1.3})
  n42 = AppendVector(A=@n40, B=Constant() {r=1.3})
  n43 = TextureSample(UVs=MaterialFunctionCall(Offset=AppendVector(A=@n38, B=@n39), Scale=StaticSwitchParameter(True=@n41, False=@n42) {parameterName=Flip}) {materialFunction=&MF_UI_Symmetry_2x1_DEPRECATED}.Result) {texture=/IcicleLarvaGameplay/Gameplay/UI/Assets/T_UI_IcicleLarva_GameplayHealthBar}
  n44 = MaterialFunctionCall(DistanceField=OneMinus(None=@n43.G), Threshold=Constant() {r=0.5}) {materialFunction=&MF_UI_AntiAliasing}
  n45 = MaterialFunctionCall(DistanceField=OneMinus(None=@n43.R), Threshold=Constant() {r=0.4}) {materialFunction=&MF_UI_AntiAliasing}
  NamedRerouteDeclaration(None=Saturate(None=Subtract(A=@n44.Inner, B=@n45.Inner))) {name=Shadow, nodeColor=<rgba 0 0 0 1>}
  NamedRerouteDeclaration(None=@n45.Inner) {name=Shape, nodeColor=<rgba 0 0 0 1>}
 ]
}
