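# World-space nav point marker widget for IcicleLarva, showing an icon with directional arrows.
# It starts collapsed, then shows only in the nearest distance band, whose squared threshold comes from a curve-table max proximity.
# Async gameplay-tag waits on the view-target pawn hide it and block distance updates while the IcicleLarva special is equipped.
# When screen-clamped, the matching edge arrow shows; otherwise only the bottom arrow shows.
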WidgetBlueprint /IcicleLarvaGameplay/Gameplay/UI/WBP_IcicleLarva_NavPointMarker
graphs [
 EventGraph {nodes [
   n1 = OnViewTargetChanged!() -> (
    OutputDelegate
    then
    PlayerController:Fort Player Controller Object Reference
    OldViewTarget:Actor Object Reference
    NewViewTarget:Actor Object Reference
   ) {is=title K2Node_CustomEvent}
   > n2 = Cast To Pawn!(execute, Object=@n1.NewViewTarget) -> (then, CastFailed, AsPawn:Pawn Object Reference) {is=title K2Node_DynamicCast}
   > n3 = Set CachedPawn(CachedPawn:Pawn Object Reference=@n2) -> (Output_Get:Pawn Object Reference)
   > n4 = WaitGameplayTagAddToActor!(
    execute:Exec
    TargetActor:Actor Object Reference=@n3
    Tag:Gameplay Tag Structure=Get IcicleLarvaSpecialEquippedTag() -> (IcicleLarvaSpecialEquippedTag:Gameplay Tag Structure)
    OnlyTriggerOnce:bool=false
   ) -> (then:Exec, AsyncTaskProxy:Ability Async Wait Gameplay Tag Added Object Reference, Added:Exec) {is=title K2Node_AsyncAction}
   > n5 = WaitGameplayTagRemoveFromActor!(
    execute:Exec
    TargetActor:Actor Object Reference=@n3
    Tag:Gameplay Tag Structure=Get IcicleLarvaSpecialEquippedTag()
    OnlyTriggerOnce:bool=false
   ) -> (then:Exec, AsyncTaskProxy:Ability Async Wait Gameplay Tag Removed Object Reference, Removed:Exec) {is=title K2Node_AsyncAction}
   Set IsLocalPlayerIcicleLarva(execute=@n5.Removed, IsLocalPlayerIcicleLarva:bool=false) -> (Output_Get:bool)
   Set IsLocalPlayerIcicleLarva(execute=@n4.Added, IsLocalPlayerIcicleLarva=true)
   > SetVisibility(InVisibility=Collapsed)
   EnableNavPointMarker() {is=title K2Node_CustomEvent}
   n6 = GetValueAtLevel!(Input:"Scalable Float Structure (by ref)"=Get NavPointMaxProximity() -> (NavPointMaxProximity:Scalable Float Structure), Level:float=0) -> (ReturnValue:float)
   > Set SqrDistanceThresholds(SqrDistanceThresholds:"Array of Float (single-precision)s"=Make Array("[0]":float=Square!(A:double=@n6) -> (ReturnValue:double)) -> (Array:"Array of Float (single-precision)s")) -> (Output_Get:"Array of Float (single-precision)s")
   Event Construct()
   > Parent: Construct()
   > SetVisibility(InVisibility=Collapsed)
   > n7 = Cast To FortPlayerController(Object=GetOwningPlayer())
   > Bind Event to On View Target Changed(self:Fort Player Controller Object Reference=@n7, Delegate=Create Event()) {is=title K2Node_AddDelegate}
   n8 = Event OnDistanceUpdate!() -> (OutputDelegate, then, PrevDistanceThreshold:int=0, NextDistanceThreshold:int=0) {is=title K2Node_Event}
   > Parent: OnDistanceUpdate!(execute, PrevDistanceThreshold:int=@n8.PrevDistanceThreshold, NextDistanceThreshold:int=@n8.NextDistanceThreshold) -> (then) {is=title K2Node_CallParentFunction}
   > n9 = Branch(Condition=Get IsLocalPlayerIcicleLarva() -> (IsLocalPlayerIcicleLarva:bool))
   UpdateWidgetVisibility!(execute=@n9.else, self, CurrentDistanceThreshold:int=@n8.NextDistanceThreshold) -> (then)
  ]}
 OnClampStateChanged {nodes [
   n1 = OnClampStateChanged!() -> (then, ClampToScreenDirection:EClampToScreenDirection Enum=Left) {is=title K2Node_FunctionEntry}
   > n2 = Branch(Condition="Not Equal (Enum)"!(A:EClampToScreenDirection Enum=@n1, B:EClampToScreenDirection Enum=None) -> (ReturnValue:bool))
   n3 = Get ClampToScreenDirection() -> (ClampToScreenDirection:EClampToScreenDirection Enum=Left)
   > SetVisibility(
    self=Get Image_ArrowLeft() -> (Image_ArrowLeft:Image Object Reference)
    InVisibility=Select(Option 0=Collapsed, Option 1=HitTestInvisible, Index="Equal (Enum)"(A:EClampToScreenDirection Enum=@n3, B:EClampToScreenDirection Enum=Left))
   )
   n4 = Get ClampToScreenDirection() -> (ClampToScreenDirection=Left)
   > SetVisibility(
    self=Get Image_ArrowRight() -> (Image_ArrowRight:Image Object Reference)
    InVisibility=Select(Option 0=Collapsed, Option 1=HitTestInvisible, Index="Equal (Enum)"(A=@n4, B=Right))
   )
   n5 = Get ClampToScreenDirection() -> (ClampToScreenDirection=Left)
   > SetVisibility(
    self=Get Image_ArrowTop() -> (Image_ArrowTop:Image Object Reference)
    InVisibility=Select(Option 0=Collapsed, Option 1=HitTestInvisible, Index="Equal (Enum)"(A=@n5, B=Top))
   )
   n6 = Get ClampToScreenDirection() -> (ClampToScreenDirection=Left)
   > SetVisibility(
    self=Get Image_ArrowBottom() -> (Image_ArrowBottom:Image Object Reference)
    InVisibility=Select(Option 0=Collapsed, Option 1=HitTestInvisible, Index="Equal (Enum)"(A=@n6, B=Bottom))
   )
   SetVisibility(execute=@n2.else, self=[Get Image_ArrowTop(), Get Image_ArrowRight(), Get Image_ArrowLeft()], InVisibility=Collapsed)
   > SetVisibility(self=Get Image_ArrowBottom(), InVisibility=HitTestInvisible)
  ]}
 UpdateWidgetVisibility {nodes [
   n1 = UpdateWidgetVisibility!() -> (then, CurrentDistanceThreshold:int) {is=title K2Node_FunctionEntry}
   > SetVisibility(InVisibility=Select(Option 0=Collapsed, Option 1=SelfHitTestInvisible, Index="Equal (Integer)"(A=@n1)))
  ]}
 VisibilityChanged {nodes [VisibilityChanged!() -> (then, Visible:bool=~) {is=title K2Node_FunctionEntry}]}
]
variables [
 CachedPawn {type=Pawn, default=null}
 IcicleLarvaSpecialEquippedTag {type=GameplayTag, default {tagName=Granted.IcicleLarva.Equipped.Special}}
 IsLocalPlayerIcicleLarva {type=boolean}
 NavPointMaxProximity {
  type=ScalableFloat
  default {
   registryType=None
   value=1
   curve {curveTable=/IcicleLarvaGameplay/DataTables/IcicleLarvaGameplayGameData, rowName=IcicleLarva.UI.NavPointProximity.Max}
  }
 }
 VisibilityChanged {}
]
widgets {
 info {inheritedWidgetCount=1, parentClass=/Script/FortniteUI.AthenaWorldUILabelWidgetBase, rootWidgetClass=&Overlay, widgetCount=10}
 widgets [
  Overlay_Bar {widgetClassPath=&Overlay, widget=Overlay_Bar}
  HorizontalBox_0 {
   widgetClassPath=&HorizontalBox
   parent=Overlay_Bar
   widget=HorizontalBox_0
   slot {class=OverlaySlot, properties {horizontalAlignment=HAlign_Center, verticalAlignment=VAlign_Center, padding=<ltrb 0 50 0 0>}}
  }
  Image_ArrowLeft {
   widgetClassPath=&Image
   bIsVariable=true
   parent=HorizontalBox_0
   widget=Image_ArrowLeft
   properties {
    visibility=Hidden
    brush {
     bIsDynamicallyLoaded=false
     drawAs=Image
     imageType=FullColor
     mirroring=NoMirror
     resourceName=None
     resourceObject=/Game/UI/CodeReferenced/T_UI_DirectionArrow
     tiling=NoTile
     imageSize=<32 32>
     margin=<ltrb 0 0 0 0>
     outlineSettings {
      bUseBrushTransparency=false
      roundingType=HalfHeightRadius
      width=0
      color {colorUseRule=UseColor_Specified, specifiedColor=<rgba 0 0 0 0>}
      cornerRadii=<0 0 0 0>
     }
     tintColor {colorUseRule=UseColor_Specified, specifiedColor=<rgba 1 1 1 1>}
     uVRegion {bIsValid=false, max=<0 0>, min=<0 0>}
    }
    renderTransform {angle=-90, scale=<1 1>, shear=<0 0>, translation=<0 0>}
   }
   slot {class=HorizontalBoxSlot, properties {horizontalAlignment=HAlign_Left, verticalAlignment=VAlign_Center}}
  }
  VerticalBox_0 {
   widgetClassPath=&VerticalBox
   parent=HorizontalBox_0
   widget=VerticalBox_0
   slot {class=HorizontalBoxSlot, properties {horizontalAlignment=HAlign_Center, verticalAlignment=VAlign_Center}}
  }
  Image_ArrowTop {
   widgetClassPath=&Image
   bIsVariable=true
   parent=VerticalBox_0
   widget=Image_ArrowTop
   properties {
    visibility=Hidden
    brush {
     bIsDynamicallyLoaded=false
     drawAs=Image
     imageType=FullColor
     mirroring=NoMirror
     resourceName=None
     resourceObject=&T_UI_DirectionArrow
     tiling=NoTile
     imageSize=<32 32>
     margin=<ltrb 0 0 0 0>
     outlineSettings {
      bUseBrushTransparency=false
      roundingType=HalfHeightRadius
      width=0
      color {colorUseRule=UseColor_Specified, specifiedColor=<rgba 0 0 0 0>}
      cornerRadii=<0 0 0 0>
     }
     tintColor {colorUseRule=UseColor_Specified, specifiedColor=<rgba 1 1 1 1>}
     uVRegion {bIsValid=false, max=<0 0>, min=<0 0>}
    }
   }
   slot {class=VerticalBoxSlot, properties {horizontalAlignment=HAlign_Center, verticalAlignment=VAlign_Top}}
  }
  VerticalBox_NavMarker {
   widgetClassPath=&VerticalBox
   parent=VerticalBox_0
   widget=VerticalBox_NavMarker
   slot {class=VerticalBoxSlot, properties {horizontalAlignment=HAlign_Center, verticalAlignment=VAlign_Center}}
  }
  Image_Icon {
   widgetClassPath=&Image
   parent=VerticalBox_NavMarker
   widget=Image_Icon
   properties {brush {
     bIsDynamicallyLoaded=false
     drawAs=Image
     imageType=FullColor
     mirroring=NoMirror
     resourceName=None
     resourceObject=/IcicleLarvaGameplay/Gameplay/UI/Assets/T_UI_IcicleLarva_Gameplay_Icon
     tiling=NoTile
     imageSize=<50 50>
     margin=<ltrb 0 0 0 0>
     outlineSettings {
      bUseBrushTransparency=false
      roundingType=HalfHeightRadius
      width=0
      color {colorUseRule=UseColor_Specified, specifiedColor=<rgba 0 0 0 0>}
      cornerRadii=<0 0 0 0>
     }
     tintColor {colorUseRule=UseColor_Specified, specifiedColor=<rgba 1 1 1 1>}
     uVRegion {bIsValid=false, max=<0 0>, min=<0 0>}
    }}
   slot {class=VerticalBoxSlot, properties {horizontalAlignment=HAlign_Center, verticalAlignment=VAlign_Center}}
  }
  Image_ArrowBottom {
   widgetClassPath=&Image
   bIsVariable=true
   parent=VerticalBox_NavMarker
   widget=Image_ArrowBottom
   properties {
    brush {
     bIsDynamicallyLoaded=false
     drawAs=Image
     imageType=FullColor
     mirroring=NoMirror
     resourceName=None
     resourceObject=&T_UI_DirectionArrow
     tiling=NoTile
     imageSize=<32 32>
     margin=<ltrb 0 0 0 0>
     outlineSettings {
      bUseBrushTransparency=false
      roundingType=HalfHeightRadius
      width=0
      color {colorUseRule=UseColor_Specified, specifiedColor=<rgba 0 0 0 0>}
      cornerRadii=<0 0 0 0>
     }
     tintColor {colorUseRule=UseColor_Specified, specifiedColor=<rgba 1 1 1 1>}
     uVRegion {bIsValid=false, max=<0 0>, min=<0 0>}
    }
    renderTransform {angle=0, scale=<1 -1>, shear=<0 0>, translation=<0 0>}
   }
   slot {class=VerticalBoxSlot, properties {horizontalAlignment=HAlign_Center, verticalAlignment=VAlign_Bottom}}
  }
  Image_ArrowRight {
   widgetClassPath=&Image
   bIsVariable=true
   parent=HorizontalBox_0
   widget=Image_ArrowRight
   properties {
    visibility=Hidden
    brush {
     bIsDynamicallyLoaded=false
     drawAs=Image
     imageType=FullColor
     mirroring=NoMirror
     resourceName=None
     resourceObject=&T_UI_DirectionArrow
     tiling=NoTile
     imageSize=<32 32>
     margin=<ltrb 0 0 0 0>
     outlineSettings {
      bUseBrushTransparency=false
      roundingType=HalfHeightRadius
      width=0
      color {colorUseRule=UseColor_Specified, specifiedColor=<rgba 0 0 0 0>}
      cornerRadii=<0 0 0 0>
     }
     tintColor {colorUseRule=UseColor_Specified, specifiedColor=<rgba 1 1 1 1>}
     uVRegion {bIsValid=false, max=<0 0>, min=<0 0>}
    }
    renderTransform {angle=90, scale=<1 1>, shear=<0 0>, translation=<0 0>}
   }
   slot {class=HorizontalBoxSlot, properties {horizontalAlignment=HAlign_Right, verticalAlignment=VAlign_Center}}
  }
  TutorialHighlight {widgetClassPath=&TutorialHighlight}
 ]
}
